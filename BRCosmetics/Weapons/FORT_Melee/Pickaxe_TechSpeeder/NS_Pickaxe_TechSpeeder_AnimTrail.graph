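# Niagara ribbon trail effect for the TechSpeeder pickaxe, built from animation-trail modules that use bone or socket index pairs.
# FN_AnimTrail spawns four trails, with spawn-group masks choosing each one's lifetime and time-of-day color, drawn by three ribbon renderers.
# Top_Edge_Trail adds a narrow, short-lived violet-blue ribbon with a camera offset.
# Both emitters fade alpha over particle life through a curve.
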NiagaraSystem /BRCosmetics/Weapons/FORT_Melee/Pickaxe_TechSpeeder/NS_Pickaxe_TechSpeeder_AnimTrail
dependencies {
 usedDataInterfaces [&NiagaraDataInterfaceCurve, &NiagaraDataInterfaceSocketReader]
 usedDynamicInputs []
 usedModules [
  &self:SystemSpawnScript.SystemScriptSource.SystemScriptGraph.NiagaraNodeAssignment_1.SetVariables_#1
  &SystemState
  &self:SystemSpawnScript.SystemScriptSource.SystemScriptGraph.NiagaraNodeAssignment_0.SetVariables_#2
 ]
}
emitters {
 FN_AnimTrail {
  emitterSpawnScript [&FN_RibbonIDOffsetUtility()]
  emitterUpdateScript [
   &EmitterState()
   &FN_SpawnAnimTrail(Number of Anim Trails=4)
   &FN_SpawnAnimTrail_v2(AnimTrailWidthMode=::FromSecond) {enabled=false}
  ]
  particleSpawnScript [
   &FN_InitializeAnimTrailBoneIndecies(First Bone or Socket Index=5, Second Bone or Socket Index=0)
   &FN_InitializeAnimTrailBoneIndecies(AnimTrail Selection=1, First Bone or Socket Index=5) {module=FN_InitializeAnimTrailBoneIndecies002}
   &FN_InitializeAnimTrailBoneIndecies(AnimTrail Selection=2, First Bone or Socket Index=5, Second Bone or Socket Index=3) {module=FN_InitializeAnimTrailBoneIndecies003}
   &FN_InitializeAnimTrailBoneIndecies(AnimTrail Selection=3, First Bone or Socket Index=5, Second Bone or Socket Index=4) {module=FN_InitializeAnimTrailBoneIndecies001}
   &FN_InitializeAnimTrail(
    Color=&MaskLinearColorBySpawnGroup(
     Masked Linear Color=&FNTimeOfDayColor(
      Night Color=<rgba 0.0301 0.0205 0.2448 1.5>
      Morning Color=<rgba 0.0301 0.0205 0.2448 1.5>
      Day Color=<rgba 0.0301 0.0205 0.2448 1>
      Evening Color=<rgba 0.0147 0.0101 0.1198 1.5>
     )
     Passthrough Linear Color=&MaskLinearColorBySpawnGroup(
      Masked Linear Color=&FNTimeOfDayColor(Night Color=<rgba 0.6551 0.7325 0.8142 0.07>, Morning Color=<rgba 1 1 1 0.03>, Day Color=<rgba 1 1 1 0.03>, Evening Color=<rgba 1 1 1 0.03>)
      Passthrough Linear Color=&MaskLinearColorBySpawnGroup(
       Masked Linear Color=&FNTimeOfDayColor(
        Night Color=<rgba 0.2679 0.2224 0.123 1>
        Morning Color=<rgba 1 0.5107 0.0795 1>
        Day Color=<rgba 1 0.5107 0.0795 1>
        Evening Color=<rgba 0.4716 0.2605 0.0745 1>
       )
       Passthrough Linear Color=&FNTimeOfDayColor(
        Night Color=<rgba 0.0302 0.0187 0.2865 1>
        Morning Color=<rgba 0.0263 0.0095 0.4022 1>
        Day Color=<rgba 0.0263 0.0095 0.4022 1>
        Evening Color=<rgba 0.0148 0.0054 0.2263 1>
       )
       Spawn Group Mask=2
      )
      Spawn Group Mask=1
     )
    )
    Lifetime=&MaskFloatBySpawnGroup(
     Masked Float=0.23
     Passthrough Float=&MaskFloatBySpawnGroup(
      Masked Float=0.2
      Passthrough Float=&MaskFloatBySpawnGroup(Masked Float=0.18, Passthrough Float=0.15, Spawn Group Mask=2)
      Spawn Group Mask=1
     )
    )
   )
   &AddVelocity(Velocity=<0 400 0>) {enabled=false}
  ]
  particleUpdateScript [
   &SolveForcesAndVelocity(Clamp Velocity=false, Limit Acceleration=false)
   &FN_UpdateAnimTrail() {module=FN_UpdateAnimTrail001}
   &ScaleColor(
    Scale Alpha=&FloatFromCurve(FloatCurve=DataInterface() {propertyValues {Curve {
        defaultValue=1
        keys [{interpMode=RCIM_Cubic, tangentMode=RCTM_User, time=0, value=1}, {interpMode=RCIM_Cubic, tangentMode=RCTM_User, time=1, value=0}]
       }}})
    Scale RGB=&Multiply_VectorByFloat(Vector=<1 1 1>, Float=&FNTimeOfDayScalar(Night Scale=0.5, Morning Scale=0.9, Evening Scale=0.9))
    ScaleA=true
    ScaleRGB=false
   )
  ]
  rendererClasses [&NiagaraRibbonRendererProperties]
  renderers [
   {rendererClass=&NiagaraRibbonRendererProperties}
   {rendererClass=&NiagaraRibbonRendererProperties, rendererIndex=1}
   {rendererClass=&NiagaraRibbonRendererProperties, rendererIndex=2}
  ]
 }
 Top_Edge_Trail {
  emitterUpdateScript [
   &EmitterState()
   &FN_SpawnAnimTrail_v2(AnimTrailWidthMode=::FromSecond)
  ]
  particleSpawnScript [
   /Game/Effects/Niagara/Modules/AnimTrail/FN_InitializeAnimTrail/FN_InitializeAnimTrailBoneIndecies(Second Bone or Socket Index=3)
   /Game/Effects/Niagara/Modules/AnimTrail/FN_InitializeAnimTrail/FN_InitializeAnimTrail(AnimTrailWidth=0.06, Color=<rgba 0.1162 0.0773 0.6568 1>, Lifetime=0.15)
  ]
  particleUpdateScript [
   &FN_UpdateAnimTrail_v2()
   &SpriteFacingAndAlignment(Sprite Facing=$Particles.RibbonFacing, WriteAlignment=false, WriteFacing=true)
   &CameraOffset(Camera Offset Amount=2)
   &ScaleColor(
    Scale Alpha=&FloatFromCurve(FloatCurve=DataInterface() {propertyValues {Curve {
        defaultValue=1
        keys [{interpMode=RCIM_Cubic, tangentMode=RCTM_User, time=0, value=1}, {interpMode=RCIM_Cubic, tangentMode=RCTM_User, time=1, value=0}]
       }}})
    ScaleA=true
    ScaleRGB=true
   ) {module=ScaleColor001}
  ]
  rendererClasses [&NiagaraRibbonRendererProperties]
  renderers [{rendererClass=&NiagaraRibbonRendererProperties}]
 }
}
summary {
 systemName=NS_Pickaxe_TechSpeeder_AnimTrail
 emitters [
  {emitterName=Top_Edge_Trail, rendererClasses [&NiagaraRibbonRendererProperties]}
  {emitterName=FN_AnimTrail, rendererClasses [&NiagaraRibbonRendererProperties]}
 ]
 userVariables [
  User.CO_ColorOverride {type=LinearColor, description="", defaultValue {struct=LinearColor, value=<rgba 1 0.3449 0 1>}}
  User.ColorSelector {type=NiagaraInt32}
  User.ColorSwitcher {type=NiagaraFloat, defaultValue=0, description=""}
  User.RibbonIDOffset {type=NiagaraInt32}
  User.SkeletalMesh {
   type=NiagaraDataInterfaceSkeletalMesh
   defaultValue {
    struct=NiagaraExt_VariableValue_DataInterface
    value {dataInterface=&self:NiagaraDataInterfaceSkeletalMesh_0}
   }
  }
 ]
}
