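# Niagara idle effect for the TechSpeeder pickaxe: a single Glow emitter that draws a soft blue sprite and a light.
# Burst-spawned particles follow a sampled skeletal mesh location with a small offset.
# A curve flickers sprite color and a sine pulses the override-tinted light, both scaled by time of day.
# A waveform animates the emitter's Power value; speed-based color scaling is present but disabled.
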NiagaraSystem /BRCosmetics/Weapons/FORT_Melee/Pickaxe_TechSpeeder/NS_Pickaxe_TechSpeeder_Idle
dependencies {
 usedDataInterfaces [&NiagaraDataInterfaceCurve]
 usedDynamicInputs []
 usedModules [
  &self:SystemSpawnScript.SystemScriptSource.SystemScriptGraph.NiagaraNodeAssignment_0.SetVariables_#1
  &SystemState
 ]
}
emitters {Glow {
  emitterSpawnScript [/SetVariables(Emitter.Depthfade=20, Emitter.Hardness=0.05, Emitter.Power=1) {isSetParameters=true}]
  emitterUpdateScript [
   &EmitterState(
    Inactive Response=::NewEnumerator2
    Life Cycle Mode=::NewEnumerator1
    Loop Behavior=::NewEnumerator1
    Loop Duration Mode=::NewEnumerator1
    UseLoopDelay=false
   )
   &SpawnBurst_Instantaneous(Use Spawn Probability=false, UseLoopCountLimit=false)
   /SetVariables(Emitter.Power:NiagaraFloat=&Add_Float(
     A=1.25
     B=&Waveform(
      Waveform Count=::NewEnumerator1
      "Waveform [1]"=::NewEnumerator0
      "[1] Phase Scale"=0.2215
      "[1] Frequency"=0.451
      "[1] Amplitude Min/Max"=<0 1>
      "[1] Scale"=0.95
      "[2] Average"=false
      "[2] Input Position"=0
      "[2] Phase"=$Emitter.Age
      "[2] Phase Scale"=0.234
      "[2] Compound Frequency"=2
      "[2] Frequency"=0.521
      "[2] Amplitude Min/Max"=<0 1>
      "[2] Scale"=1
      "[2] Bias"=0
      "[2] Exponent"=1
      "[2] Pulse Duration"=0.1
      "[2] Clip Triangle Min/Max"=<0 0>
      "[2] Random Seed"=0
      "[3] Phase"=$Emitter.Age
     )
    )) {isSetParameters=true}
  ]
  particleSpawnScript [
   &InitializeParticle(
    Color=&HueShiftLinearColor(
     Hue=0
     Saturation=1
     Value=1
     Linear Color=&MultiplyLinearColor(
      Linear Color=&HueShiftLinearColor(Hue=0, Saturation=1, Value=1.5, Linear Color=<rgba 0.0919 0.0586 0.8374 1>)
      Scale Factor=<1 1 1 0.05>
     )
    )
    Position Offset=<0 0 50>
    Sprite Size=<150 150>
    Sprite Size Mode=::NewEnumerator4
    UsePositionOffset=false
    Write Material Random=true
   )
   &CameraOffset()
  ]
  particleUpdateScript [
   &ParticleState(
    Kill Particles When Lifetime Has Elapsed=false
    Let Infinitely Lived Particles Die When Emitter Deactivates=true
    Loop Particles Lifetime=true
   )
   &SkeletalMeshLocation(
    Inherit Velocity=false
    "Inherit Velocity (Scale)"=0
    Interpolate Mesh Sampling=false
    MaskSpawnGroup=false
    Mesh Sampling Type=::NewEnumerator5
    Sampled Position Offset=<0 0 15>
    Sampled Position Offset Coordinate Space:ENiagaraMeshTransforms=::NewEnumerator2
    Skeletal Mesh=$System.Skeletal Mesh
    Socket Sampling Mode=::NewEnumerator5
   ) {module=SkeletalMeshLocation001}
   &ScaleColor(
    Scale RGB=&Multiply_VectorByFloat(
     Vector=<1 1 1>
     Float=&FloatFromCurve(
      FloatCurve=DataInterface() {propertyValues {Curve {keys [<at 0 0.8>,<at 0.1 0.9>,<at 0.3 0.7>,<at 0.4 0.8>,<at 0.5 0.8>,<at 0.6 1>,<at 0.8 0.75>,<at 0.9 0.95>,<at 1 0.8>]}}}
      Scale Curve=&Multiply_Float(A=&FNTimeOfDayScalar(Night Scale=0.5, Morning Scale=0.7, Day Scale=0.8, Evening Scale=0.7), B=1)
     )
    )
    ScaleA=true
    ScaleRGB=true
   )
   &FN_Light_Attributes(
    Color=&MultiplyLinearColor(
     Linear Color=&HueShiftLinearColor(Hue=0, Saturation=1, Value=10, Linear Color=$User.CO_ColorOverride)
     Scale Factor=&MakeVector4_FromFloat(Float=&Multiply_Float(
       A=&FNTimeOfDayScalar(Night Scale=0.2, Morning Scale=0.2, Day Scale=0.2, Evening Scale=0.2)
       B=&Sine(Normalized Angle=$Emitter.Age, Period=0.2, Scale=0.04, Bias=0.8)
      ))
    )
    Create Color Attribute=true
    Create Exponent Attribute=true
    Exponent=4
    Scale Color=0.5
   )
   /Niagara/Modules/Update/Color/ScaleColorBySpeed(Max Speed Threshold=500, Min Speed Threshold=0, RGB Scale Min=0.1, Source Velocity:Vector3f=$Engine.Owner.Velocity) {enabled=false}
  ]
  rendererClasses [&NiagaraSpriteRendererProperties, &NiagaraLightRendererProperties]
  renderers [
   {rendererClass=&NiagaraSpriteRendererProperties}
   {rendererClass=&NiagaraLightRendererProperties, rendererIndex=1}
  ]
 }}
summary {
 systemName=NS_Pickaxe_TechSpeeder_Idle
 emitters [{emitterName=Glow, rendererClasses [&NiagaraSpriteRendererProperties, &NiagaraLightRendererProperties]}]
 userVariables [
  User.bIsFrontEndPreview {type=NiagaraBool, defaultValue=true}
  User.CO_ColorOverride {type=LinearColor, description="", defaultValue {struct=LinearColor, value=<rgba 0.0595 0.0497 0.2789 1>}}
  User.SkeletalMesh {
   type=NiagaraDataInterfaceSkeletalMesh
   defaultValue {
    struct=NiagaraExt_VariableValue_DataInterface
    value {dataInterface=&self:NiagaraDataInterfaceSkeletalMesh_0}
   }
  }
 ]
}
